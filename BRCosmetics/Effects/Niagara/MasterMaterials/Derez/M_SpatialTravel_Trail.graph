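# Unlit, two-sided translucent Niagara material for sprites, ribbons and meshes, drawing a blocky derez trail for a spatial travel effect.
# Scrolling Perlin noise on grid-quantized UVs, offset by a dynamic parameter, blends two colors into emissive, tinted by particle color.
# Stepped gradients, time-animated hashed cells and grid textures build named masks for trail and dissolve, with an alternate dissolve blend.
# Several experimental node chains feed nothing.
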Material /BRCosmetics/Effects/Niagara/MasterMaterials/Derez/M_SpatialTravel_Trail
graph {
 properties {
  bAutomaticallySetUsageInEditor=false
  bUsedWithNiagaraMeshParticles=true
  bUsedWithNiagaraRibbons=true
  bUsedWithNiagaraSprites=true
  blendMode=BLEND_Translucent
  opacityMaskClipValue=0.1
  shadingModel=MSM_Unlit
  twoSided=true
  usedShadingModels=Unlit
 }
 nodes [
  n1 = ScalarParameter() {defaultValue=4, parameterName=Grid}
  n2 = Floor(None=Multiply(A=TextureCoordinate(), B=@n1))
  n3 = Subtract(A=Divide(A=@n1)) {constB=0.5}
  n4 = Length(None=Subtract(A=Divide(A=@n2, B=@n1), B=Divide(A=@n3) {constB=10}) {constB=0.45}) {is=class MaterialExpressionLength}
  n5 = Step(Y=@n4) {constX=0.5}
  Ceil(None=@n5)
  n6 = Multiply(A=TextureCoordinate(), B=@n1)
  n7 = ComponentMask(None=TextureCoordinate()) {g=true}
  n8 = Subtract(A=Multiply(A=Frac(None=@n7)) {constB=2})
  n9 = Floor(None=Multiply(A=Abs(None=@n8)) {constB=2})
  n10 = Power(Base=OneMinus(None=Divide(A=@n9))) {constExponent=4}
  n11 = Multiply(A=MaterialFunctionCall(Vec2=Ceil(None=@n6)) {materialFunction=&MF_Hash}.Vec3, B=@n10)
  n12 = Multiply(A=Time()) {constB=0.5}
  n13 = ConstantBiasScale(None=Sine(None=Add(A=@n12, B=@n11)))
  n14 = OneMinus(None=@n4)
  n15 = Power(Base=Multiply(A=@n14, B=@n5)) {constExponent=1}
  ComponentMask(None=LinearInterpolate(A=Multiply(A=@n11, B=@n13), B=@n15, Alpha=@n14)) {r=true}
  n16 = Multiply(A=Time()) {constB=0.5}
  n17 = TextureCoordinate() {uTiling=0.5, vTiling=0.5}
  n18 = Multiply(A=Time(), B=Constant2Vector() {r=0.1})
  n19 = ScalarParameter() {defaultValue=4, parameterName=Grid}
  n20 = Frac(None=Multiply(A=Add(A=@n17, B=@n18), B=@n19))
  n21 = TextureSample(UVs=Add(A=@n17, B=Divide(A=@n20, B=@n19))) {
   samplerSource=SSM_Wrap_WorldGroupSettings
   samplerType=SAMPLERTYPE_LinearColor
   texture=/Game/Packages/Fortress_SharedMaterials/Textures/T_Slowfield_GridSquares_01
  }
  ConstantBiasScale(None=Sine(None=Add(A=@n16, B=@n21.R)))
  Constant() {r=4}
  Constant() {r=4}
  Constant() {r=4}
  Constant() {r=4}
  MaterialFunctionCall(Emissive=NamedRerouteUsage()) {materialFunction=&MF_FX_EmissiveAdjustments}
  MaterialFunctionCall() {materialFunction=&DebugTimeSine}
  n22 = NamedRerouteUsage()
  Multiply(A=NamedRerouteUsage(), B=@n22)
  Multiply(A=@n15, B=@n10)
  Multiply(A=@n22, B=NamedRerouteUsage())
  n23 = Multiply(B=@n21.R) {constA=1}
  NamedRerouteDeclaration(None=LinearInterpolate(A=@n23, Alpha=NamedRerouteUsage())) {name=Derez_Mask, nodeColor=<rgba 0.0076 0.75 1 1>}
  n24 = Ceil(None=Multiply(A=TextureCoordinate(), B=ScalarParameter() {defaultValue=4, parameterName=Grid}))
  n25 = Time() {period=1}
  n26 = Frac(None=Add(
    A=MaterialFunctionCall(Vec2=@n24) {materialFunction=&MF_Hash}.Vec2
    B=Multiply(A=@n25) {constB=0.5}
   ))
  n27 = LinearInterpolate(A=ComponentMask(None=Sine(None=@n26)) {r=true}, B=NamedRerouteDeclaration(None=@n10) {name=Circle_Mask, nodeColor=<rgba 0.2656 0 1 1>}) {constAlpha=0.8}
  NamedRerouteDeclaration(None=Multiply(A=Power(Base=@n27) {constExponent=4}) {constB=5}) {name=MaskTrail, nodeColor=<rgba 1 0.2344 0 1>}
  n28 = Power(Base=Multiply(A=NamedRerouteUsage()) {constB=0.2}) {constExponent=5}
  n29 = TextureCoordinate()
  n30 = ScalarParameter() {defaultValue=4, parameterName=Grid}
  n31 = Divide(A=Floor(None=Multiply(A=@n29, B=@n30)), B=@n30)
  n32 = Multiply(A=Time(), B=Constant2Vector() {r=-2})
  n33 = DynamicParameter() {defaultValue=<rgba 0 1 1 1>, paramNames [UV_Offset,Param2,Param3,Param4]}
  n34 = Constant()
  n35 = TextureSample(UVs=Add(A=Add(A=@n31, B=@n32), B=AppendVector(A=@n33.UV_Offset, B=@n34))) {samplerSource=SSM_Wrap_WorldGroupSettings, texture=&/T_PerlinNoise_Pack}
  n36 = NamedRerouteUsage()
  n37 = Multiply(A=@n35.B, B=@n36)
  n38 = Saturate(None=Multiply(A=Multiply(A=@n28, B=@n37)) {constB=5})
  n39 = Saturate(None=@n35.R)
  n40 = LinearInterpolate(
   A=VectorParameter() {group=Derez, parameterName=Color_01, defaultValue=<rgba 0.0214 0.114 0.5 1>}.RGB
   B=VectorParameter() {group=Derez, parameterName=Color_02, defaultValue=<rgba 0 1.9099 3 1>}.RGB
   Alpha=@n39
  )
  n41 = LinearInterpolate(A=@n40, B=Multiply(A=@n40) {constB=5}, Alpha=@n39)
  NamedRerouteDeclaration(None=Multiply(A=Add(A=@n38, B=@n41), B=ParticleColor().RGB)) {name=Emissive, nodeColor=<rgba 0 0.3437 1 1>}
  n42 = TextureCoordinate()
  n43 = ScalarParameter() {defaultValue=4, parameterName=Grid}
  n44 = LinearInterpolate(Alpha=TextureSample(UVs=Multiply(A=@n42, B=@n43)) {samplerType=SAMPLERTYPE_LinearColor, texture=/Game/Environments/Hexmap/HexSelection/T_Hex_RadarGrid02}.R) {constA=-2, constB=10}
  NamedRerouteDeclaration(None=OneMinus(None=Saturate(None=@n44))) {name=Grid, nodeColor=<rgba 1 0 0.5859 1>}
  n45 = TextureCoordinate()
  n46 = ScalarParameter() {defaultValue=4, parameterName=Grid}
  n47 = Multiply(A=OneMinus(None=ComponentMask(None=@n45) {r=true}), B=@n46)
  n48 = Saturate(None=Divide(A=Floor(None=@n47), B=@n46))
  NamedRerouteDeclaration(None=@n48) {name=U_Gradient, nodeColor=<rgba 0.1484 1 0 1>}
  n49 = Multiply(A=TextureCoordinate(), B=Constant() {r=8})
  n50 = TextureCoordinate() {uTiling=0.5, vTiling=0.5}
  n51 = Ceil(None=Abs(None=@n50))
  n52 = Add(A=Multiply(A=Time() {period=1}) {constB=0.5}, B=Add(A=ComponentMask(None=@n51) {r=true}, B=ComponentMask(None=@n51) {g=true}))
  n53 = Add(A=MaterialFunctionCall(Vec2=Ceil(None=@n49)) {materialFunction=&MF_Hash}.Vec2, B=@n52)
  n54 = Multiply(A=Sine(None=Frac(None=@n53)), B=@n48)
  NamedRerouteDeclaration(None=ComponentMask(None=LinearInterpolate(A=@n54, Alpha=@n48) {constB=0.5}) {r=true}) {name=Gradient_Dissolve, nodeColor=<rgba 0 1 0.625 1>}
  n55 = OneMinus(None=NamedRerouteUsage())
  n56 = NamedRerouteUsage()
  n57 = OneMinus(None=Saturate(None=Divide(A=@n55, B=@n56)))
  n58 = Saturate(None=SmoothStep(Min=Subtract(A=@n35.R, B=@n57), Value=@n56))
  n59 = Add(A=@n58, B=LinearInterpolate(B=@n58, Alpha=@n36))
  n60 = Power(Base=Multiply(A=NamedRerouteUsage(), B=NamedRerouteUsage()), Exp=ScalarParameter() {defaultValue=5, parameterName=Power Dissolve}) {constExponent=5}
  NamedRerouteDeclaration(None=LinearInterpolate(A=@n59, B=Multiply(A=@n60, B=@n59), Alpha=ScalarParameter() {parameterName=UseAltDissolve})) {name=Derez_Dissolve, nodeColor=<rgba 0.4766 1 0 1>}
  NamedRerouteDeclaration(None=ParticleColor().A) {name=AlphaDissolve, nodeColor=<rgba 1 0 0.5156 1>}
  Saturate(None=Multiply(A=@n23) {constB=1.2})
  ScalarParameter() {defaultValue=4, parameterName=LinesScale}
  n61 = Multiply(A=ScalarParameter() {defaultValue=0.5, parameterName=Sine Speed}, B=@n52)
  n62 = TextureSample(UVs=@n50) {samplerType=SAMPLERTYPE_LinearColor, texture=&T_Slowfield_GridSquares_01}
  Sine(None=Frac(None=Add(A=@n61, B=@n62.R)))
  Step() {constX=0.5}
 ]
}
